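# Animation Blueprint for the TameEagle backpack cosmetic, adding secondary motion to its pie, croissant, donut and straps.
# The AnimGraph additively offsets the three pastry base bones, then chains spring-driven, angle-limited Anim Dynamics nodes in root-relative space.
# The EventGraph computes gravity override from a parameter asset, a falling-not-skydiving blend and a front-end alpha.
# The dynamics nodes show constant gravity and alpha, with no visible reads of those computed values.
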AnimBlueprint /BRCosmetics/Accessories/FORT_Backpacks/Backpack_F_MED_TameEagle/Meshes/F_MED_TameEagle_Pack_AnimBP
target_skeleton=&/F_MED_TameEagle_Pack_Skeleton
graphs [
 AnimGraph {nodes [
   n1 = "Transform (Modify) Bone - Bone: base_dyn_pie"!(
    ComponentPose=Local To Component(LocalPose=Input Pose())
    bAlphaBoolEnabled:bool=True
    Translation="-1.38,0.8,0"
    Rotation="0,10,0"
    Scale="1,1,1"
   ) -> (Pose) {
    alphaInputType=Bool
    rotationMode=BMM_Additive
    rotationSpace=BCS_BoneSpace
    translationMode=BMM_Additive
    translationSpace=BCS_BoneSpace
    alphaBoolBlend {blendInTime=0.5, blendOption=Linear, blendOutTime=0.2}
    boneToModify {boneName=base_dyn_pie}
    rotation {yaw=10}
    translation=<-1.38 0.8>
   }
   n2 = "Transform (Modify) Bone - Bone: base_dyn_croissant"!(ComponentPose=@n1, bAlphaBoolEnabled:bool=True, Translation="-2.08,-0.05,0", Rotation="0,10,0", Scale="1,1,1") -> (Pose) {
    alphaInputType=Bool
    rotationMode=BMM_Additive
    rotationSpace=BCS_BoneSpace
    translationMode=BMM_Additive
    translationSpace=BCS_BoneSpace
    alphaBoolBlend {blendInTime=0.5, blendOption=Linear, blendOutTime=0.2}
    boneToModify {boneName=base_dyn_croissant}
    rotation {yaw=10}
    translation=<-2.08 -0.05>
   }
   n3 = "Transform (Modify) Bone - Bone: base_dyn_donut"!(ComponentPose=@n2, bAlphaBoolEnabled:bool=True, Translation="-2.58,0.5,0", Rotation="0,10,0", Scale="1,1,1") -> (Pose) {
    alphaInputType=Bool
    rotationMode=BMM_Additive
    rotationSpace=BCS_BoneSpace
    translationMode=BMM_Additive
    translationSpace=BCS_BoneSpace
    alphaBoolBlend {blendInTime=0.5, blendOption=Linear, blendOutTime=0.2}
    boneToModify {boneName=base_dyn_donut}
    rotation {yaw=10}
    translation=<-2.58 0.5>
   }
   n4 = Anim Dynamics - Bone: dyn_croissant_00!(ComponentPose=@n3, Alpha=1, GravityOverride:Vector="0,0,-980", ExternalForce="0,0,0") -> (Pose) {
    angularDampingOverride=1
    angularSpringConstant=10000
    bAngularSpring=true
    bOverrideAngularDamping=true
    bUseGravityOverride=true
    gravityScale=0.35
    simulationSpace=RootRelative
    boundBone {boneName=dyn_croissant_00}
    componentLinearAccScale=<0.3 0.3 0.4>
    componentLinearVelScale=<0.3 0.3 0.4>
    gravityOverride {z=-980}
    physicsBodyDefinitions [{
      boundBone {boneName=dyn_croissant_00}
      boxExtents=<20 30 30>
      constraintSetup {angularLimitsMax=<15 5 3>, angularLimitsMin=<-15 -5 -3>, angularTarget=<1 0 0>}
      localJointOffset=<4 0 0>
     }]
   }
   n5 = Anim Dynamics - Bone: dyn_donut_00!(ComponentPose=@n4, Alpha=1, GravityOverride:Vector="0,0,-980", ExternalForce="0,0,0") -> (Pose) {
    angularDampingOverride=1
    angularSpringConstant=20000
    bAngularSpring=true
    bOverrideAngularDamping=true
    bUseGravityOverride=true
    gravityScale=0.35
    simulationSpace=RootRelative
    boundBone {boneName=dyn_donut_00}
    componentLinearAccScale=<0.3 0.3 0.4>
    componentLinearVelScale=<0.3 0.3 0.4>
    gravityOverride {z=-980}
    physicsBodyDefinitions [{
      boundBone {boneName=dyn_donut_00}
      boxExtents=<20 30 30>
      constraintSetup {angularLimitsMax=<10 3 0>, angularLimitsMin=<-10 -3 0>, angularTarget=<1 0 0>}
      localJointOffset=<4 0 0>
     }]
   }
   n6 = Anim Dynamics - Bone: dyn_pie_00!(ComponentPose=@n5, Alpha=1, GravityOverride:Vector="0,0,-980", ExternalForce="0,0,0") -> (Pose) {
    angularDampingOverride=1
    angularSpringConstant=15000
    bAngularSpring=true
    bOverrideAngularDamping=true
    bUseGravityOverride=true
    gravityScale=0.35
    simulationSpace=RootRelative
    boundBone {boneName=dyn_pie_00}
    componentLinearAccScale=<0.3 0.3 0.4>
    componentLinearVelScale=<0.3 0.3 0.4>
    gravityOverride {z=-980}
    physicsBodyDefinitions [{
      boundBone {boneName=dyn_pie_00}
      boxExtents=<20 30 30>
      constraintSetup {angularLimitsMax=<15 5 5>, angularLimitsMin=<-15 -5 -5>, angularTarget=<1 0 0>}
      localJointOffset=<4 0 0>
     }]
   }
   n7 = Anim Dynamics - Bone: dyn_strapInside_00!(ComponentPose=@n6, Alpha=1, GravityOverride:Vector="0,0,-980", ExternalForce="0,0,0") -> (Pose) {
    angularDampingOverride=1
    angularSpringConstant=30000
    bAngularSpring=true
    bOverrideAngularDamping=true
    bUseGravityOverride=true
    gravityScale=0.35
    simulationSpace=RootRelative
    boundBone {boneName=dyn_strapInside_00}
    componentLinearAccScale=<0.3 0.3 0.4>
    componentLinearVelScale=<0.3 0.3 0.4>
    gravityOverride {z=-980}
    physicsBodyDefinitions [{
      boundBone {boneName=dyn_strapInside_00}
      boxExtents=<20 30 30>
      constraintSetup {angularLimitsMax=<0 0.5 8>, angularLimitsMin=<0 -0.5 0>, angularTarget=<1 0 0>}
      localJointOffset=<4 0 0>
     }]
   }
   n8 = Anim Dynamics - Bone: dyn_strapOutside_00!(ComponentPose=@n7, Alpha=1, GravityOverride:Vector="0,0,-980", ExternalForce="0,0,0") -> (Pose) {
    angularDampingOverride=0.99
    angularSpringConstant=6000
    bAngularSpring=true
    bOverrideAngularDamping=true
    bUseGravityOverride=true
    gravityScale=0.35
    simulationSpace=RootRelative
    boundBone {boneName=dyn_strapOutside_00}
    componentLinearAccScale=<0.6 0.6 0.7>
    componentLinearVelScale=<0.6 0.6 0.7>
    gravityOverride {z=-980}
    physicsBodyDefinitions [{
      boundBone {boneName=dyn_strapOutside_00}
      boxExtents=<20 30 30>
      constraintSetup {angularLimitsMax=<0 0 0>, angularLimitsMin=<0 0 -175>, angularTarget=<1 0 0>}
      localJointOffset=<4 0 0>
     }]
   }
   n9 = Anim Dynamics - Bone: dyn_strapOutside_01!(ComponentPose=@n8, Alpha=1, GravityOverride:Vector="0,0,-980", ExternalForce="0,0,0") -> (Pose) {
    angularDampingOverride=1
    angularSpringConstant=6000
    bAngularSpring=true
    bOverrideAngularDamping=true
    bUseGravityOverride=true
    gravityScale=0.35
    simulationSpace=RootRelative
    boundBone {boneName=dyn_strapOutside_01}
    componentLinearAccScale=<0.6 0.6 0.7>
    componentLinearVelScale=<0.6 0.6 0.7>
    gravityOverride {z=-980}
    physicsBodyDefinitions [{
      boundBone {boneName=dyn_strapOutside_01}
      boxExtents=<20 30 30>
      constraintSetup {angularLimitsMax=<0 0 0>, angularLimitsMin=<0 0 -150>, angularTarget=<1 0 0>}
      localJointOffset=<4 0 0>
     }]
   }
   n10 = Anim Dynamics - Bone: dyn_strapOutside_02!(ComponentPose=@n9, Alpha=1, GravityOverride:Vector="0,0,-980", ExternalForce="0,0,0") -> (Pose) {
    angularDampingOverride=1
    angularSpringConstant=6000
    bAngularSpring=true
    bOverrideAngularDamping=true
    bUseGravityOverride=true
    gravityScale=0.35
    simulationSpace=RootRelative
    boundBone {boneName=dyn_strapOutside_02}
    componentLinearAccScale=<0.6 0.6 0.7>
    componentLinearVelScale=<0.6 0.6 0.7>
    gravityOverride {z=-980}
    physicsBodyDefinitions [{
      boundBone {boneName=dyn_strapOutside_02}
      boxExtents=<20 30 30>
      constraintSetup {angularLimitsMax=<0 0 0>, angularLimitsMin=<0 0 -175>, angularTarget=<1 0 0>}
      localJointOffset=<4 0 0>
     }]
   }
   Output Pose(Result=Component To Local(ComponentPose=@n10)) {layerGroup=None}
  ]}
 EventGraph {nodes [
   Event BlueprintUpdateAnimation()
   > n1 = Sequence!(execute) -> (then_0, then_1, then_2:Exec)
   > n2 = "Gravity Override State Function (Parameter Asset) (Nativized)"(
    CharacterPartAnimInstance=Self-Reference()
    GameTimeInSeconds=0
    Parameters=&/F_MED_TameEagle_Pack_Parameters
   )
   > Set Gravity Override(Gravity Override=@n2)
   n3 = Sequence!(execute=@n1.then_1) -> (then_0, then_1)
   n4 = Get bIsSkyDiving()
   > n5 = Branch(Condition=AND Boolean(A=Get bIsFalling(), B=NOT Boolean(A=@n4)))
   > Set Is Falling Not Skydiving(Is Falling Not Skydiving:bool=true) -> (Output_Get:bool)
   Set Is Falling Not Skydiving(execute=@n5.else, Is Falling Not Skydiving=false)
   n6 = Get Is Falling Not Skydiving() -> (Is Falling Not Skydiving:bool)
   Set Is Falling Float(
    execute=@n3.then_1
    Is Falling Float:double=FInterpTo(Current=Get Is Falling Float() -> (Is Falling Float:double), Target="To Float (Boolean)"(InBool=@n6), DeltaTime=0, InterpSpeed=-1)
   ) -> (Output_Get:double)
   n7 = Branch(execute=@n1.then_2, Condition=Get bIsInFrontEnd())
   > Set FEAlpha(FEAlpha:double=0.3) -> (Output_Get:double)
   Set FEAlpha(execute=@n7.else, FEAlpha=1)
   TryGetPawnOwner()
  ]}
]
variables [
 FEAlpha {type=number}
 Gravity Override {type=Vector}
 Is Falling Float {type=number}
 Is Falling Not Skydiving {type=boolean}
]
